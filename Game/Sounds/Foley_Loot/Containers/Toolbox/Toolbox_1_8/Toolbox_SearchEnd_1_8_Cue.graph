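# Sound cue that plays when searching a toolbox loot container ends.
# A mixer layers the shared no-more-loot UI sound with one of three equally weighted toolbox search-end recordings.
# Only the random toolbox layer passes through a modulator for slight pitch and volume variation.
# The cue uses mono item-interact attenuation and overrides concurrency to cap simultaneous instances.
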SoundCue /Game/Sounds/Foley_Loot/Containers/Toolbox/Toolbox_1_8/Toolbox_SearchEnd_1_8_Cue
root=Mixer_0
nodes [
 {id=Mixer_0, children [WavePlayer_0,Modulator_0], properties {inputVolume [0.8,0.8]}}
 {id=Modulator_0, children [Random_0], properties {pitchMax=1.07, pitchMin=0.92, volumeMax=1.07, volumeMin=0.92}}
 {id=Random_0, children [WavePlayer_4,WavePlayer_5,WavePlayer_6], properties {weights [1,1,1]}}
 {id=WavePlayer_0, properties {soundWaveAssetPtr=/Game/Sounds/Fort_UI_Sounds/Gameplay/Fort_UI_NoMoreLoot}}
 {id=WavePlayer_4, properties {soundWaveAssetPtr=&/Toolbox_SearchEnd6}}
 {id=WavePlayer_5, properties {soundWaveAssetPtr=&/Toolbox_SearchEnd8}}
 {id=WavePlayer_6, properties {soundWaveAssetPtr=&/Toolbox_SearchEnd11}}
]
properties {
 attenuationSettings=/Game/Sounds/Attenuation/Fort_Item_Interact_Mono_Att
 bOverrideConcurrency=true
 maxDistance=1500
 volumeMultiplier=0.7
 concurrencyOverrides {maxCount=10}
}
