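# Sound cue for searching a toolbox loot container, routed through the Fort_Loot_Collect sound class with mono item-interact attenuation.
# A random node picks from three search waves and feeds a modulator that slightly varies pitch and volume.
# The mixer plays that shared chain twice, once immediately and once quieter after a roughly one-second random delay, layering a longer rummage from few samples.
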SoundCue /Game/Sounds/Foley_Loot/Containers/Toolbox/Toolbox_1_8/Toolbox_Search_1_8_Cue
root=Mixer_0
nodes [
 {id=Delay_0, children [Modulator_0], properties {delayMax=1.2, delayMin=1}}
 {id=Mixer_0, children [Modulator_0,Delay_0], properties {inputVolume [1,0.7]}}
 {id=Modulator_0, children [Random_0], properties {pitchMax=1.07, pitchMin=0.92, volumeMax=1.07, volumeMin=0.92}}
 {id=Random_0, children [WavePlayer_3,WavePlayer_4,WavePlayer_5], properties {weights [1,1,1]}}
 {id=WavePlayer_3, properties {soundWaveAssetPtr=&/Toolbox_Search8}}
 {id=WavePlayer_4, properties {soundWaveAssetPtr=&/Toolbox_Search13}}
 {id=WavePlayer_5, properties {soundWaveAssetPtr=&/Toolbox_Search16}}
]
properties {
 attenuationSettings=/Game/Sounds/Attenuation/Fort_Item_Interact_Mono_Att
 bOverrideConcurrency=true
 maxDistance=1500
 soundClassObject=&Fort_Loot_Collect
 volumeMultiplier=0.8141
 concurrencyOverrides {maxCount=10}
}
